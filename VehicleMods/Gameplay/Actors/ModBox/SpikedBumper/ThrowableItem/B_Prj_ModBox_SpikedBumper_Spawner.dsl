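; Thrown projectile for the Spiked Bumper vehicle mod box that spawns a building gameplay actor (BGA) where it lands.
; On the server, valid hits stop movement and spawn the BGA, while invalid hits can revert to a pickup.
; Water entry checks spawn settings and blocking volumes, and spawn height snaps to the water surface.
; Spawning can align to the surface normal, applies the weapon's cosmetic wrap and notifies the owner by gameplay event.
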(event Collision|EventHit
  (params (MyComp :type Primitive%20Component%20Object%20Reference) (Other :type Actor%20Object%20Reference) (OtherComp :type Primitive%20Component%20Object%20Reference) (bSelfMoved :type Boolean) (HitLocation :type Vector :default "0, 0, 0") (HitNormal :type Vector :default "0, 0, 0") (NormalImpulse :type Vector :default "0, 0, 0") (Hit :type Hit%20Result%20Structure%20%28by%20ref%29))
  ;[0 1 0.1338] Event Hit
  ; Check if the projectile has hit a valid target and determine the behavior of the hit
  (|CallParentFunction MyComp Other OtherComp bSelfMoved HitLocation HitNormal NormalImpulse Hit)
  (bind _inairloop (Variables|Default|GetInAirLoop)
    (:then
      (Audio|Components|Audio|FadeOut _inairloop 0.25 0.0 "Linear" :!class /Script/Engine.AudioComponent))
    (:else
      _))
  (Networking|SwitchHasAuthority
    (:Authority
      (Utilities|IsHitValidforSpawn OtherComp Other Hit
        (:ValidSpawn
          (CallFunction|InitiateSpawnBGA Hit Other OtherComp HitLocation))
        (:InvalidSpawnAndRevertToPickup
          (CallFunction|RevertThrownObjecttoPickup))))))

(event Custom|SpawnCue
  (GameplayCue|ExecuteGameplayCue%28Burst%29Local self (Variables|Default|GetGameplay%20Cue%20Tag) (Ability|GameplayCue|MakeGameplayCueParameters 0.0)))

(event Custom|InitiateSpawnBGA
  (params (HitResult :type Hit%20Result%20Structure) (HitOtherActor :type Actor%20Object%20Reference) (HitOtherComp :type Primitive%20Component%20Object%20Reference) (HitLocation :type Vector))
  (bind (_bblockinghit _binitialoverlap _time _distance _location _impactpoint _normal _impactnormal _physmat _hitactor _hitcomponent) (Collision|BreakHitResult HitResult))
  (bind (_asfort_aipawn _bsuccess) (Utilities|Casting|CastToFortAIPawn HitOtherActor))
  ;[0 0.009347 1] InitiateSpawnBGA
  ; Setup (stop simulating, do once) and then Spawn BGA on valid hit
  (Game|Components|ProjectileMovement|StopSimulating (Variables|Movement|GetProjectileMovementComponent) HitResult)
  (CallFunction|SpawnBGA _normal _hitactor _hitcomponent (Math|Vector|SelectVector HitLocation (CallFunction|GetClosestPoint HitOtherComp (Transformation|GetActorLocation)) _bsuccess) false HitResult)
  (CallFunction|SpawnCue))

(event Custom|RevertThrownObjectToPickup
  ;[0.3021 0.1664 0.007888] RevertThrownObjectToPickup
  (Networking|SwitchHasAuthority
    (:Authority
      (bind _itemdeftospawnaspickupwheninvalid (Variables|Spawning|GetItemDefToSpawnAsPickupWhenInvalid)
        (:then
          (Inventory|SpawnPickupInWorld _itemdeftospawnaspickupwheninvalid 1 (Transformation|GetActorLocation) "0.000000,0.000000,1.000000" 0 true false false -1 "Tossed" "TossedByPlayer" 0 false))
        (:else
          ;;[1 0 0] Warning: Item expected to revert to a pickup, but the Item Definition is invalid
          (Fortnite|FortShippingLog self (Utilities|String|Append (Utilities|GetDisplayName self) ": attempt to spawn thrown projectile as pickup failed (invalid ItemDef) - destroying projectile") true)))
      (Rendering|SetActorHiddenInGame (Variables|Default|GetSetActorHiddenOnSpawnBGA))
      (if (Variables|Default|GetDestroyActorImmediaetlyOnRevertToObject)
        (Actor|DestroyActor)))))

(event OnEnterWaterEvent(WaterInteractionComponent)
  (params (WaterBody :type Fort%20Water%20Body%20Actor%20Object%20Reference) (WaterInteractionComponent :type Fort%20Water%20Interaction%20Component%20Object%20Reference) (bIsFirstBody :type Boolean))
  ;[0 1 0.1338] Spawn on Enter Water
  ; Check water spawning conditions and either spawn or revert the item to a pickup
  (Utilities|FlowControl|Sequence
    (:then_0
      (Networking|SwitchHasAuthority
        (:Authority
          (if (Variables|Spawning|ValidityChecks|GetbSpawnOnCollision)
            (if (Variables|Spawning|ValidityChecks|GetbAllowSpawnOnWater)
              (if (not (CallFunction|CheckIsBlockedbyVolume))
                (CallFunction|SpawnBGA "0, 0, 0" WaterBody 0 (Transformation|GetActorLocation) true))
              (else
                _))
            (CallFunction|RevertThrownObjecttoPickup)))))
    (:then_1
      (if bIsFirstBody
        (CallFunction|HeldWaterImpact)))))

(event Custom|Held%20Water%20Impact
  ;[0.6195 0.07955 1] HeldWaterImpact
  ; SFX and VFX on water impact - play splash sound
  (Audio|PlaySoundatLocation (Variables|Spawning|FX|GetSplash%20Sound) (Transformation|GetActorLocation)))

(fn CheckForWater_
  (params (CurrPos :type Vector) (HitActor :type Actor%20Object%20Reference))
  (bind _asfort_water_body_actor (Utilities|Casting|CastToFortWaterBodyActor HitActor)
    (:then
      (bind _outwatersurfaceinfo (WaterBody|GetWaterSurfaceInfoStruct _asfort_water_body_actor (Transformation|GetActorLocation) true false false))
      (bind _waterplanelocation (Utilities|Struct|BreakWaterSurfaceInfo _outwatersurfaceinfo))
      (bind (_waterplanelocation_x _waterplanelocation_y _waterplanelocation_z) (Math|Vector|BreakVector _waterplanelocation))
      (return (Math|Vector|MakeVector (.x CurrPos) (.y CurrPos) (+ _waterplanelocation_z (Variables|Spawning|TransformData|GetSpawnPosZOffsetInWater)))))
    (:CastFailed
      (return CurrPos))))

(fn CheckIsBlockedByVolume
  ;[0.474 0.6502 1] CheckIsBlockedByVolume
  ; If the object is set to allow blocking volumes to prevent spawning and determine if an overlap is blocking it
  (bind _boverlapping (Default|IsOverlappingDeployableBlockVolume self (Variables|Spawning|ValidityChecks|GetBlockingAreaCheckRadius) (Transformation|GetActorLocation)))
  (return (select (Variables|Spawning|ValidityChecks|GetbCanSpawnInBlockVolume) (not (Variables|Spawning|ValidityChecks|GetbCanSpawnInBlockVolume)) _boverlapping)))

(fn GetClosestPoint
  (params (Component :type Primitive%20Component%20Object%20Reference) (StartingPoint :type Vector))
  (bind (_outpointonbody _returnvalue) (Collision|GetClosestPointonCollision Component StartingPoint "None"))
  (return (Math|Vector|SelectVector (+ (* (Math|Vector|Normalize (- StartingPoint _outpointonbody)) 18.0) _outpointonbody) (Transformation|GetActorLocation) (> _returnvalue 0.0))))

(fn SpawnBGA_
  (params (Normal :type Vector) (HitActor :type Actor%20Object%20Reference) (HitComp :type Primitive%20Component%20Object%20Reference) (Pos :type Vector) (FromWater :type Boolean) (Hit :type Hit%20Result%20Structure))
  (bind (_returnvalue_x _returnvalue_y _returnvalue_z) (Math|Vector|BreakVector (Transformation|GetActorLocation (Variables|Spawning|GetSpawnedBGA))))
  (if (Variables|Spawning|GetSpawned)
    (if (> (.z Pos) _returnvalue_z)
      (Actor|DestroyActor (Variables|Spawning|GetSpawnedBGA)))
    (else
      (Variables|Spawning|SetSpawned true)
      (Utilities|FlowControl|Sequence
        (:then_0
          (Actor|SetLifeSpan (Variables|Default|GetLifespanAfterSpawnBGA))
          (Collision|SetActorEnableCollision false)
          (Rendering|SetActorHiddenInGame (Variables|Default|GetSetActorHiddenOnSpawnBGA))
          (Audio|PlaySoundatLocation (Variables|Spawning|FX|GetSoundActivate) (Transformation|GetActorLocation)))
        (:then_1
          _))))
  (bind _returnvalue (Building|SpawnBuildingActor (Variables|Spawning|GetBGAToSpawn) (Math|Transform|MakeTransform (Math|Vector|vector+vector (CallFunction|CheckforWater Pos HitActor) (* (Utilities|MakeSureNormalIsntZero Normal) (Variables|Spawning|TransformData|GetSpawnPosOffsetAlongNormal)) (Math|Vector|MakeVector 0.0 0.0 (Variables|Spawning|TransformData|GetSpawnPosAdditionalZOffset))) (Math|Rotator|SelectRotator (Math|Rotator|MakeRotfromXZ (Math|Vector|ProjectVectorontoPlane (Transformation|GetActorForwardVector (Game|GetInstigator)) (Utilities|MakeSureNormalIsntZero Normal)) (Utilities|MakeSureNormalIsntZero Normal)) (Transformation|GetActorRotation self) (Variables|Spawning|TransformData|GetUseImpactNormalForBGASpawnRotation)) (Variables|Spawning|TransformData|GetSpawnScale)) (Variables|Actor|GetInstigator) (Variables|Actor|GetInstigator)))
  (Variables|Spawning|SetSpawnedBGA _returnvalue)
  (bind _asbuilding_gameplay_actor (Utilities|Casting|CastToBuildingGameplayActor _returnvalue)
    (:then
      (Building|InitializeKismetSpawnedBuildingActor _asbuilding_gameplay_actor 0 0 true 0 false)
      (Utilities|FlowControl|Sequence
        (:then_0
          (bind _asbga_athena_with_gravity_parent (Utilities|Casting|CastToBGA_Athena_WithGravity_Parent _asbuilding_gameplay_actor)))
        (:then_1
          (bind _returnvalue_2 (CallFunction|GetOwnerWeapon))
          (CosmeticWraps|AttemptApplyOverrideWraptoActor _returnvalue_2 _asbuilding_gameplay_actor))
        (:then_2
          _)))
    (:CastFailed
      _))
  (Ability|SendGameplayEventtoActor (Actor|GetOwner) "(TagName=\"Event.Projectile.SpawnedBGA\")" (Utilities|Struct|MakeGameplayEventData :OptionalObject _returnvalue :ContextHandle "()" :InstigatorTags "(GameplayTags=)" :TargetTags "(GameplayTags=)" :EventMagnitude 0.0 :TargetData "()"))
  (return _returnvalue))

(fn ConstructionScript)

(fn WillBounceOffTarget
  (params (Hit :type Hit%20Result%20Structure%20%28by%20ref%29))
  (bind (_bblockinghit _binitialoverlap _time _distance _location _impactpoint _normal _impactnormal _physmat _hitactor _hitcomponent) (Collision|BreakHitResult Hit))
  (if (not (Components|ComponentHasTag _hitcomponent "NoStick"))
    (bind _returnvalue (GameplayTags|HasMatchingGameplayTag(Message) _hitactor "(TagName=\"Athena.Item.ShieldBubbleSafeZone\")"))
    (if (not _returnvalue)
      (Utilities|Casting|CastToBuildingGameplayActor _hitactor
        (:then
          (return true))
        (:CastFailed
          _)))
    (return false)))
